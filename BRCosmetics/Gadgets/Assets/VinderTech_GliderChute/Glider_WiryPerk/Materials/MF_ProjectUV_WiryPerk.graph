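# Material function for the WiryPerk glider that generates planar-projected UVs from pre-skinned local position.
# A cube of UVSize centered on Root_Position plus per-axis offsets normalizes the vertex-interpolated position, with optional clamping to 0–1.
# It outputs a YZ side projection plus XY top and XZ back projections.
# FlipUV mirrors the side projection's U on the side at or above the root's X, handling left and right separately.
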MaterialFunction /BRCosmetics/Gadgets/Assets/VinderTech_GliderChute/Glider_WiryPerk/Materials/MF_ProjectUV_WiryPerk
graph {nodes [
  n1 = VertexInterpolator(VS=LocalPosition() {localOrigin=InstancePreSkinning}.XYZ)
  n2 = FunctionInput() {bUsePreviewValueAsDefault=true, inputName=Root_Position}
  MaterialExpressionAdd_2 = Add(
   A=@n2
   B=MaterialFunctionCall(
    X=FunctionInput() {bUsePreviewValueAsDefault=true, inputName=CenterOffsetX, inputType=FunctionInput_Scalar, sortPriority=1}
    Y=FunctionInput() {bUsePreviewValueAsDefault=true, inputName=CenterOffsetY, inputType=FunctionInput_Scalar, sortPriority=2}
    Z=FunctionInput() {bUsePreviewValueAsDefault=true, inputName=CenterOffsetZ, inputType=FunctionInput_Scalar, sortPriority=3}
   ) {materialFunction=&MakeFloat3}.Result
  )
  MaterialExpressionFunctionInput_0 = FunctionInput(Preview=Constant() {r=30}) {bUsePreviewValueAsDefault=true, inputName=UVSize, inputType=FunctionInput_Scalar}
  n3 = Custom!(Center=@MaterialExpressionAdd_2, Size=@MaterialExpressionFunctionInput_0) {
   code="// Calculate half the length\r\n    float halfSize = Size * 0.5;\r\n    float3 V00;      // UV00\r\n    V00 = Center + float3(-halfSize, -halfSize, -halfSize);\r\n    return V00;"
   description=Get the 00 position
   showCode=true
   inputs [
    {
     inputName=Center
     input {expression=&self:MaterialExpressionAdd_2}
    }
    {
     inputName=Size
     input {expression=&self:MaterialExpressionFunctionInput_0}
    }
   ]
  }
  n4 = Divide(A=Subtract(A=@n1.PS, B=@n3), B=@MaterialExpressionFunctionInput_0)
  n5 = LinearInterpolate(
   A=@n4
   B=Clamp(None=@n4)
   Alpha=FunctionInput(Preview=Constant() {r=1}) {bUsePreviewValueAsDefault=true, inputName=UseClamp?, inputType=FunctionInput_Scalar}
  )
  n6 = OneMinus(None=ComponentMask(None=@n5) {b=true, g=true})
  n7 = OneMinus(None=ComponentMask(None=@n6) {r=true})
  n8 = ComponentMask(None=@n6) {g=true}
  n9 = AppendVector(A=OneMinus(None=ComponentMask(None=@n6) {r=true}), B=ComponentMask(None=@n6) {g=true})
  n10 = Constant() {r=1}
  n11 = If(
   A=Subtract(A=ComponentMask(None=@n1.PS) {r=true}, B=ComponentMask(None=@n2) {r=true})
   B=Constant()
   "A > B"=@n10
   "A == B"=@n10
   "A < B"=Constant()
  )
  n12 = FunctionInput() {bUsePreviewValueAsDefault=true, inputName=FlipUV, inputType=FunctionInput_Scalar, previewValue=<0 0 0 0>}
  FunctionOutput(None=LinearInterpolate(A=AppendVector(A=@n7, B=@n8), B=LinearInterpolate(A=@n9, B=@n6, Alpha=@n11), Alpha=@n12))
  FunctionOutput(None=OneMinus(None=ComponentMask(None=@n5) {g=true, r=true})) {outputName=UV_TopProjection}
  FunctionOutput(None=OneMinus(None=ComponentMask(None=@n5) {b=true, r=true})) {outputName=UV_BackProjection}
 ]}
