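# Material function that adds animated emissive color effects to the WiryPerk glider's material attributes.
# Fan color comes from top-projected, rotating, hue-cycling radial gradients, with a pre-skinning local X test providing a left/right mask.
# Side color runs a panned, sine-offset gradient through a curve atlas or gradient texture.
# Texture masks, a Fresnel term and a base emissive map feed the layering, and an input mask channel controls the final attribute blend.
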MaterialFunction /BRCosmetics/Gadgets/Assets/VinderTech_GliderChute/Glider_WiryPerk/Materials/MF_AdvancedEmissive_WiryPerk
graph {nodes [
  n1 = NamedRerouteUsage()
  n2 = Constant()
  n3 = Multiply(A=NamedRerouteUsage(), B=NamedRerouteUsage())
  n4 = GetMaterialAttributes(None=FunctionInput() {inputName=Material, inputType=FunctionInput_MaterialAttributes}) {attributeGetTypes [#1, #2, #3]}
  n5 = LinearInterpolate(A=@n4.Roughness, Alpha=NamedRerouteUsage()) {constB=0.5}
  n6 = NamedRerouteUsage()
  FunctionOutput(None=BlendMaterialAttributes(
    A=NamedRerouteUsage()
    B=SetMaterialAttributes!(MaterialAttributes=@n1, BaseColor=@n2, EmissiveColor=@n3, Roughness=@n5) {attributeSetTypes [#2, #1, #3]}
    Alpha=ChannelMaskParameter(None=@n6) {group=05_Tech Art Masks, parameterName=AdvancedEmissiveMaskChannel}.R
   ) {vertexAttributeBlendType=UseA})
  Max()
  NamedRerouteDeclaration(None=FunctionInput() {inputName=Alpha}) {name=FXMask, nodeColor=<rgba 0 0.5781 1 1>}
  n7 = NamedRerouteUsage()
  n8 = TextureSampleParameter2D() {
   group=WiryPerk
   parameterName=BaseEmissive
   texture=/BRCosmetics/Gadgets/Assets/VinderTech_GliderChute/Glider_WiryPerk/Texture/T_WiryPerk_Glider_E
  }
  n9 = Max(A=Max(A=@n8.R, B=@n8.R), B=@n8.B)
  NamedRerouteDeclaration(None=SetMaterialAttributes(MaterialAttributes=NamedRerouteUsage(), BaseColor=LinearInterpolate(A=@n7, Alpha=@n9) {constB=0}, EmissiveColor=@n8.RGB) {attributeSetTypes [#2, #1, #3]}) {name=Material Attributes, nodeColor=<rgba 1 0.5156 0 1>}
  NamedRerouteDeclaration(None=@n4.BaseColor) {name=InputBaseColor, nodeColor=<rgba 1 0.2344 0 1>}
  n10 = Fresnel(ExponentIn=Constant() {r=1}) {baseReflectFraction=0}
  n11 = Power(
   Base=LinearInterpolate(
    A=OneMinus(None=@n10)
    B=@n10
    Alpha=ScalarParameter() {group=06_Advanced Emissive, parameterName=InvertAdvancedEmissiveFresnel, sortPriority=39}
   )
   Exp=ScalarParameter() {defaultValue=2, group=06_Advanced Emissive, parameterName=AdvancedEmissiveFresnelPower, sortPriority=39}
  )
  NamedRerouteDeclaration(None=LinearInterpolate(A=Constant() {r=0.5}, B=Constant() {r=1.5}, Alpha=Saturate(None=@n11))) {name=Emissive Fresnel, nodeColor=<rgba 0 0.0859 1 1>}
  n12 = TextureSample() {samplerType=SAMPLERTYPE_Masks, texture=/BRCosmetics/Gadgets/Assets/VinderTech_GliderChute/Glider_WiryPerk/Texture/T_WiryPerk_Glider_TA_Mask}
  NamedRerouteDeclaration(None=LinearInterpolate(A=NamedRerouteUsage(), B=NamedRerouteUsage(), Alpha=@n12.R)) {name=AdvancedEmissiveOutput, nodeColor=<rgba 1 0 0.2578 1>}
  n13 = ComponentMask(None=LocalPosition() {localOrigin=InstancePreSkinning}.XYZ) {r=true}
  n14 = Constant() {r=1}
  NamedRerouteDeclaration(None=If(A=VertexInterpolator(VS=@n13).PS, "A > B"=@n14, "A == B"=@n14, "A < B"=Constant())) {name=LeftRightMask, nodeColor=<rgba 0.4766 1 0 1>}
  n15 = ScalarParameter() {defaultValue=0.18, group=WiryPerk, parameterName=color_cycling_Speed}
  n16 = Constant() {r=58.4067}
  n17 = ScalarParameter() {defaultValue=-51.1989, group=WiryPerk, parameterName=CenterOffsetX}
  n18 = MaterialFunctionCall!(
   Root_Position=Constant()
   UseClamp?
   FlipUV
   UVSize=If(A=NamedRerouteUsage(), "A > B"=Constant() {r=90}, "A == B"=@n16, "A < B"=@n16) {constB=0.3}
   CenterOffsetX=LinearInterpolate(A=@n17, B=OneMinus(None=@n17), Alpha=NamedRerouteUsage())
   CenterOffsetY=Constant() {r=-4.4558}
   CenterOffsetZ=Constant()
  ) {materialFunction=&/MF_ProjectUV_WiryPerk}
  n19 = ScalarParameter() {defaultValue=-1.4, group=WiryPerk, parameterName=Rotation Angle}
  n20 = LinearInterpolate(A=@n19, B=Multiply(A=@n19) {constB=-1}, Alpha=NamedRerouteUsage())
  n21 = NamedRerouteUsage()
  n22 = MaterialFunctionCall(UVs=@n18.UV_TopProjection, "Rotation Angle (0-1)"=Multiply(A=Time(), B=LinearInterpolate(A=@n20, Alpha=@n21) {constB=0.15})) {materialFunction=&CustomRotator}
  n23 = MaterialFunctionCall(
   Hue Shift Percentage=Multiply(A=Time(), B=@n15) {constB=0.1}
   Texture=TextureSample(UVs=@n22.Rotated Values) {texture=/BRCosmetics/Gadgets/Assets/VinderTech_GliderChute/Glider_WiryPerk/Texture/T_Gradient_Radial_Red}.RGB
  ) {materialFunction=&HueShift}
  n24 = MaterialFunctionCall(
   Hue Shift Percentage=Multiply(A=Time(), B=@n15) {constB=0.1}
   Texture=TextureSample(UVs=@n22.Rotated Values) {texture=/BRCosmetics/Gadgets/Assets/VinderTech_GliderChute/Glider_WiryPerk/Texture/T_Gradient_Radial_Blue}.RGB
  ) {materialFunction=&HueShift}
  n25 = NamedRerouteUsage()
  n26 = Power(Base=TextureSample(UVs=@n18.UV_TopProjection) {texture=/Game/Global/Textures/Default/Masks_and_Gradients/T_New_Soft_Flair}.R)
  n27 = Multiply(
   A=LinearInterpolate(
    A=LinearInterpolate(A=@n23.Result, B=@n24.Result, Alpha=@n25)
    B=TextureSample(UVs=@n22.Rotated Values) {texture=/BRCosmetics/Gadgets/Assets/VinderTech_GliderChute/Glider_WiryPerk/Texture/T_WiryPerk_color_RadialPNG}.RGB
    Alpha=NamedRerouteUsage()
   )
   B=LinearInterpolate(A=LinearInterpolate(Alpha=@n26) {constA=0.65, constB=5}, Alpha=NamedRerouteUsage())
  )
  n28 = Multiply(A=@n27) {constB=0.5}
  n29 = Constant()
  NamedRerouteDeclaration(None=LinearInterpolate(A=@n27, B=Desaturation(None=@n28, Fraction=@n29), Alpha=NamedRerouteUsage())) {name=Fan_RGB, nodeColor=<rgba 1 0 0.7969 1>}
  n30 = Constant()
  n31 = ComponentMask(None=ScalarParameter() {defaultValue=1, group=WiryPerk, parameterName=EmissivePannerScale}) {
   desc="Kendall added this node on 5/27/2020.\r\nPreviously, this just piped the 3-vector value from\r\nthe texture sample straight into the UV's, which\r\ncaused the material to fail compile"
   g=true
   r=true
  }
  n32 = Multiply(
   A=StaticSwitchParameter(True=TextureCoordinate() {coordinateIndex=1}, False=TextureCoordinate()) {group=06_Advanced Emissive, parameterName=UseSecondUVChannelForPanning}
   B=@n31
  )
  n33 = ScalarParameter() {group=WiryPerk, parameterName=GradientSpeed_X, sortPriority=2}
  n34 = ScalarParameter() {defaultValue=0.2, group=WiryPerk, parameterName=GradientSpeed_Y, sortPriority=2}
  n35 = TextureObjectParameter() {
   group=06_Advanced Emissive
   parameterName=EmissiveGradient
   samplerType=SAMPLERTYPE_Masks
   sortPriority=1
   texture=/Game/Global/Textures/Default/Masks_and_Gradients/T_VertGradient_MASK
  }
  n36 = TextureSample(UVs=Panner(Coordinate=Add(A=@n30, B=@n32), Speed=AppendVector(A=@n33, B=@n34)), Tex=@n35) {
   samplerSource=SSM_Wrap_WorldGroupSettings
   samplerType=SAMPLERTYPE_Masks
   texture=/Game/Weapons/WeaponSkins/Textures/T_Swirly_Noise_highContrast
  }
  n37 = ScalarParameter() {defaultValue=0.5, group=WiryPerk, parameterName=ColorOffset}
  n38 = Multiply(A=Time()) {constB=0.1}
  n39 = ScalarParameter() {defaultValue=0.5, group=WiryPerk, parameterName=SineSpeed, sortPriority=2}
  n40 = ConstantBiasScale(None=Sine(None=Multiply(A=@n38, B=@n39)))
  n41 = Multiply(A=Add(A=@n36.R, B=LinearInterpolate(B=@n37, Alpha=@n40) {constB=0.5})) {constB=0.5}
  n42 = CurveAtlasRowParameter(CurveTime=ComponentMask(None=@n41) {g=true}) {
   atlas=&/WiryPerk_ColorAtlas
   curve=&/WiryPerk_CurveBase
   group=06_Advanced Emissive
   parameterName=PanningEmissiveColorCurve
   sortPriority=6
  }
  n43 = TextureSampleParameter2D(UVs=ComponentMask(None=@n41) {g=true, r=true}) {
   group=06_Advanced Emissive
   parameterName=PanningEmissiveColorGradient
   samplerSource=SSM_Wrap_WorldGroupSettings
   sortPriority=6
   texture=/Game/Global/Textures/Default/Masks_and_Gradients/3pxRGBImage
  }
  NamedRerouteDeclaration(None=StaticSwitchParameter(True=Multiply(A=@n42) {constB=0.99}, False=@n43.RGB) {group=06_Advanced Emissive, parameterName=UseCurveAtlasForGradient, sortPriority=0}) {name=SideRGB, nodeColor=<rgba 0 1 0.8828 1>}
  n44 = TextureSample() {samplerType=SAMPLERTYPE_Masks, texture=&T_WiryPerk_Glider_TA_Mask}
  NamedRerouteDeclaration(None=@n44.R) {name=TA_MaskR, nodeColor=<rgba 1 0.3984 0 1>}
  NamedRerouteDeclaration(None=@n44.G) {name=TA_MaskG, nodeColor=<rgba 1 0.3984 0 1>}
  NamedRerouteDeclaration(None=@n44.B) {name=TA_MaskB, nodeColor=<rgba 1 0.3984 0 1>}
  NamedRerouteDeclaration(None=@n4.MaterialAttributes) {name=MA, nodeColor=<rgba 0 1 0.5547 1>}
 ]}
